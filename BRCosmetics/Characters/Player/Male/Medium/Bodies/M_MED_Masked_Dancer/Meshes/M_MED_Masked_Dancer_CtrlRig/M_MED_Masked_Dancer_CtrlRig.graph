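# Control Rig for the M_MED_Masked_Dancer male body that procedurally offsets its side pouch as the lower spine rotates.
# It reads spine_02 relative to pelvis, converts the rotation to Euler and remaps the clamped X angle into a blend weight.
# That weight scales a small fixed additive local rotation on dyn_sidePouch_mstr, carrying its child zipper bones along.
# The other strap and jacket bones are left untouched.
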ControlRigBlueprint /BRCosmetics/Characters/Player/Male/Medium/Bodies/M_MED_Masked_Dancer/Meshes/M_MED_Masked_Dancer_CtrlRig/M_MED_Masked_Dancer_CtrlRig
graphs [
 RigVMFunctionLibrary {links [], nodes []}
 RigVMModel {nodes [
   RigUnit_BeginExecution = &RigUnit_BeginExecution()
   n1 = &RigUnit_GetRelativeBoneTransform(Bone=spine_02, Space=pelvis)
   /Script/ControlRig.RigUnit_ModifyBoneTransforms(
    BoneToModify=((Bone="dyn_sidePouch_mstr",Transform=(Rotation=(X=0,Y=0,Z=0,W=1),Translation=(X=0,Y=0,Z=0),Scale3D=(X=1,Y=1,Z=1))))
    Mode=AdditiveLocal
    WeightMaximum=1
    Weight=&RigVMFunction_MathFloatRemap(
     SourceMaximum=5
     bClamp=true
     Value=&RigVMFunction_MathQuaternionToEuler(RotationOrder=ZYX, Value=@n1.Transform.Rotation).Result.X
    ).Result
    BoneToModify.0.Transform.Rotation=&RigVMFunction_MathQuaternionFromEuler(Euler=(X=0,Y=-7,Z=0)).Result
    ExecutePin=@RigUnit_BeginExecution.ExecutePin
   ) {is=struct RigVMUnitNode}
  ]}
]
hierarchy [
 root {loc=<0 0 0>, rot=<0 0 0 1>}
 pelvis {parent=root, loc=<0 -0.7071 93.1191>, rot=<0.0127 0.707 -0.0127 -0.707>}
 spine_01 {parent=pelvis, loc=<6.7494 -0.6388 0>, rot=<0 0 0.034 -0.9994>}
 spine_02 {parent=spine_01, loc=<7.2518 -0.035 0>, rot=<0 0 0.0238 -0.9997>}
 spine_03 {parent=spine_02, loc=<6.0104 0 0>, rot=<0 0 0.0149 -0.9999>}
 spine_04 {parent=spine_03, loc=<8.1024 0 0>, rot=<0 0 -0.0132 -0.9999>}
 spine_05 {parent=spine_04, loc=<10.0146 0 0>, rot=<0 0 -0.0015 -1>}
 clavicle_l {parent=spine_05, loc=<4.3575 -0.1466 -2.3223>, rot=<0.0465 -0.7289 -0.009 -0.6829>}
 upperarm_l {parent=clavicle_l, loc=<15.5111 0 0>, rot=<0.0226 -0.3234 -0.0088 -0.9459>}
 lowerarm_l {parent=upperarm_l, loc=<26.6514 0 0>, rot=<0.0391 -0.0099 -0.2453 -0.9686>}
 hand_l {parent=lowerarm_l, loc=<22.3491 0 0>, rot=<-0.7081 -0.0507 0.0975 -0.6975>}
 index_metacarpal_l {parent=hand_l, loc=<1.6535 0.0183 -2.96>, rot=<0.0158 -0.0124 -0.0173 -0.9996>}
 index_01_l {parent=index_metacarpal_l, loc=<8.3008 0.0357 -0.0524>, rot=<-0.0044 0.0485 0.0924 -0.9945>}
 index_02_l {parent=index_01_l, loc=<3.3587 -0.3004 0.0452>, rot=<-0.002 0.0027 0.3225 -0.9466>}
 index_03_l {parent=index_02_l, loc=<2.5578 -0.123 -0.0252>, rot=<-0.002 0.0067 0.2558 -0.9667>}
 middle_metacarpal_l {parent=hand_l, loc=<2.1856 -0.1725 -0.5984>, rot=<0.006 0.0168 -0.0329 -0.9993>}
 middle_01_l {parent=middle_metacarpal_l, loc=<7.3275 -0.2447 0.0129>, rot=<-0.0208 0.0042 0.0858 -0.9961>}
 middle_02_l {parent=middle_01_l, loc=<4.0123 0.1401 0.1175>, rot=<-0.0218 0.0258 0.3512 -0.9357>}
 middle_03_l {parent=middle_02_l, loc=<2.6077 0.0534 0.1373>, rot=<0.0082 0.0188 0.1708 -0.9851>}
 pinky_metacarpal_l {parent=hand_l, loc=<1.7616 -0.9753 3.3814>, rot=<-0.0397 0.0814 -0.069 -0.9935>}
 pinky_01_l {parent=pinky_metacarpal_l, loc=<7.1709 -0.4831 -0.1094>, rot=<-0.0654 -0.0035 0.2242 -0.9723>}
 pinky_02_l {parent=pinky_01_l, loc=<2.8274 0.0893 -0.1043>, rot=<0.0037 0.0011 0.2854 -0.9584>}
 pinky_03_l {parent=pinky_02_l, loc=<2.2046 0.0307 -0.0078>, rot=<-0.0002 0.0098 0.1445 -0.9895>}
 ring_metacarpal_l {parent=hand_l, loc=<2.2074 -0.3625 1.3638>, rot=<-0.0284 0.0722 -0.0458 -0.9959>}
 ring_01_l {parent=ring_metacarpal_l, loc=<7.0295 -0.6197 -0.1446>, rot=<-0.0489 -0.0252 0.1892 -0.9804>}
 ring_02_l {parent=ring_01_l, loc=<3.6174 0.4646 -0.1027>, rot=<0 0 0.2806 -0.9598>}
 ring_03_l {parent=ring_02_l, loc=<2.4152 0.1141 0.0199>, rot=<0 0 0.1189 -0.9929>}
 thumb_01_l {parent=hand_l, loc=<1.3118 -0.3191 -2.459>, rot=<0.6546 -0.4042 0.0664 -0.6354>}
 thumb_02_l {parent=thumb_01_l, loc=<5.9336 -0.0448 0.0401>, rot=<-0.0037 0.0213 0.298 -0.9543>}
 thumb_03_l {parent=thumb_02_l, loc=<2.2223 0.3682 -0.0907>, rot=<0.0118 -0.013 0.0066 -0.9998>}
 lowerarm_twist_01_l {parent=lowerarm_l, loc=<17.8549 0 0>, rot=<0 0 0 1>}
 lowerarm_twist_02_l {parent=lowerarm_l, loc=<9.6199 0 0>, rot=<0 0 0 1>}
 upperarm_twist_01_l {parent=upperarm_l, loc=<0.0041 0 0>, rot=<0 0 0 -1>}
 upperarm_twist_02_l {parent=upperarm_l, loc=<13.7941 0 0>, rot=<0 0 0 -1>}
 clavicle_r {parent=spine_05, loc=<4.3575 -0.1466 2.3223>, rot=<0.7289 0.0465 -0.6829 0.009>}
 upperarm_r {parent=clavicle_r, loc=<-15.5111 0 0>, rot=<0.0226 -0.3234 -0.0088 -0.9459>}
 lowerarm_r {parent=upperarm_r, loc=<-26.6514 0 0>, rot=<0.0391 -0.0099 -0.2453 -0.9686>}
 hand_r {parent=lowerarm_r, loc=<-22.3492 0 0>, rot=<-0.7081 -0.0507 0.0975 -0.6975>}
 index_metacarpal_r {parent=hand_r, loc=<-1.6534 -0.0181 2.9601>, rot=<0.0158 -0.0124 -0.0173 -0.9996>}
 index_01_r {parent=index_metacarpal_r, loc=<-8.3008 -0.0357 0.0524>, rot=<-0.0044 0.0485 0.0924 -0.9945>}
 index_02_r {parent=index_01_r, loc=<-3.3589 0.3003 -0.0453>, rot=<-0.002 0.0027 0.3225 -0.9466>}
 index_03_r {parent=index_02_r, loc=<-2.5578 0.1231 0.0251>, rot=<-0.002 0.0067 0.2558 -0.9667>}
 middle_metacarpal_r {parent=hand_r, loc=<-2.1856 0.1726 0.5984>, rot=<0.006 0.0168 -0.0329 -0.9993>}
 middle_01_r {parent=middle_metacarpal_r, loc=<-7.3273 0.2447 -0.0129>, rot=<-0.0208 0.0042 0.0858 -0.9961>}
 middle_02_r {parent=middle_01_r, loc=<-4.0142 -0.1411 -0.07>, rot=<-0.0217 0.0259 0.3512 -0.9357>}
 middle_03_r {parent=middle_02_r, loc=<-2.6074 -0.0534 -0.1312>, rot=<0.0083 0.0188 0.1708 -0.9851>}
 pinky_metacarpal_r {parent=hand_r, loc=<-1.7617 0.9753 -3.3814>, rot=<-0.0397 0.0814 -0.069 -0.9935>}
 pinky_01_r {parent=pinky_metacarpal_r, loc=<-7.1707 0.4832 0.1094>, rot=<-0.0654 -0.0035 0.2242 -0.9723>}
 pinky_02_r {parent=pinky_01_r, loc=<-2.8274 -0.0892 0.1043>, rot=<0.0037 0.0011 0.2854 -0.9584>}
 pinky_03_r {parent=pinky_02_r, loc=<-2.2045 -0.031 0.0078>, rot=<-0.0002 0.0098 0.1445 -0.9895>}
 ring_metacarpal_r {parent=hand_r, loc=<-2.2073 0.3627 -1.3638>, rot=<-0.0284 0.0722 -0.0458 -0.9959>}
 ring_01_r {parent=ring_metacarpal_r, loc=<-7.0297 0.6197 0.1446>, rot=<-0.0489 -0.0252 0.1892 -0.9804>}
 ring_02_r {parent=ring_01_r, loc=<-3.6172 -0.4648 0.1027>, rot=<0 0 0.2806 -0.9598>}
 ring_03_r {parent=ring_02_r, loc=<-2.4152 -0.1141 -0.0199>, rot=<0 0 0.1189 -0.9929>}
 thumb_01_r {parent=hand_r, loc=<-1.3119 0.3192 2.459>, rot=<0.6546 -0.4042 0.0664 -0.6354>}
 thumb_02_r {parent=thumb_01_r, loc=<-5.9335 0.0447 -0.0401>, rot=<-0.0037 0.0213 0.298 -0.9543>}
 thumb_03_r {parent=thumb_02_r, loc=<-2.2223 -0.3682 0.0907>, rot=<0.0118 -0.013 0.0066 -0.9998>}
 lowerarm_twist_01_r {parent=lowerarm_r, loc=<-17.855 0 0>, rot=<0 0 0 -1>}
 lowerarm_twist_02_r {parent=lowerarm_r, loc=<-9.6199 0 0>, rot=<0 0 0 -1>}
 upperarm_twist_01_r {parent=upperarm_r, loc=<-0.0041 0 0>, rot=<0 0 0 1>}
 upperarm_twist_02_r {parent=upperarm_r, loc=<-13.7942 0 0>, rot=<0 0 0 1>}
 neck_01 {parent=spine_05, loc=<11.3138 0.1529 0>, rot=<0 0 -0.2003 -0.9797>}
 neck_02 {parent=neck_01, loc=<4.5627 0 0>, rot=<0 0 0.0504 -0.9987>}
 head {parent=neck_02, loc=<5.3914 0 0>, rot=<0 0 0.115 -0.9934>}
 dyn_chestStrap_mstr {parent=spine_05, loc=<6.8437 7.8272 -9.6947>, rot=<-0.2712 -0.9357 0.08 0.2111>}
 dyn_chestStrap1 {parent=dyn_chestStrap_mstr, loc=<0 0 0>, rot=<0 0 0 -1>}
 dyn_chestStrap2 {parent=dyn_chestStrap1, loc=<7.7543 0 0>, rot=<0.0047 0.0081 0.1127 -0.9936>}
 dyn_chestStrap3 {parent=dyn_chestStrap2, loc=<3.6515 0 0>, rot=<0 0 0 1>}
 dyn_sidePouch_mstr {parent=spine_01, loc=<2.1946 -0.2657 -3.849>, rot=<0.0135 -0.8368 -0.0088 -0.5473>}
 dyn_sidePouch {parent=dyn_sidePouch_mstr, loc=<0 0 0>, rot=<0 0 0 1>}
 dyn_frontZipper_mstr {parent=dyn_sidePouch, loc=<-1.2501 12.0847 2.6099>, rot=<0.4626 -0.1976 -0.558 -0.66>}
 dyn_frontZipper {parent=dyn_frontZipper_mstr, loc=<0 0 0>, rot=<0 0 0 1>}
 dyn_backZipper_mstr {parent=dyn_sidePouch, loc=<6.6724 -10.896 0.7694>, rot=<-0.2678 -0.3098 0.4419 -0.7981>}
 dyn_backZipper {parent=dyn_backZipper_mstr, loc=<0 0 0>, rot=<0 0 0 1>}
 thigh_l {parent=pelvis, loc=<-2.5693 -0.0109 -7.7501>, rot=<0.0012 0.0572 0.0155 -0.9982>}
 calf_l {parent=thigh_l, loc=<-39.0484 0 0>, rot=<0.0006 -0.0161 -0.0391 -0.9991>}
 calf_twist_01_l {parent=calf_l, loc=<-35.8131 -0.0967 0.0397>, rot=<0.0237 0.0095 -0.0035 -0.9997>}
 calf_twist_02_l {parent=calf_l, loc=<-18.8064 -0.0967 0.0397>, rot=<0.0237 0.0095 -0.0035 -0.9997>}
 foot_l {parent=calf_l, loc=<-42.561 0.1077 0.4292>, rot=<0.0418 -0.0455 0.0343 -0.9975>}
 ball_l {parent=foot_l, loc=<-6.1599 13.8264 0.6806>, rot=<0.0001 -0.0001 -0.7063 -0.7079>}
 thigh_twist_01_l {parent=thigh_l, loc=<-21.6303 0 0>, rot=<0.0013 0 0 -1>}
 dyn_legStrap1_mstr_l {parent=thigh_twist_01_l, loc=<0.0494 0.088 -8.9399>, rot=<0.022 -0.9942 0.0015 -0.105>}
 dyn_legStrap1_l {parent=dyn_legStrap1_mstr_l, loc=<0 0 0>, rot=<0 0 0 1>}
 dyn_legStrap2_l {parent=dyn_legStrap1_l, loc=<5.5135 0 0>, rot=<0.0001 -0.0143 -0.0409 -0.9991>}
 dyn_legStrap3_l {parent=dyn_legStrap2_l, loc=<4.0917 0 0>, rot=<-0.0005 -0.0186 0.0054 -0.9998>}
 dyn_legStrap4_l {parent=dyn_legStrap3_l, loc=<3.6853 0 0>, rot=<0.0328 -0.17 0.2075 -0.9628>}
 dyn_legStrap5_l {parent=dyn_legStrap4_l, loc=<3.4233 0 0>, rot=<-0.0074 0.0199 -0.0472 -0.9987>}
 dyn_legStrap6_l {parent=dyn_legStrap5_l, loc=<2.7048 0 0>, rot=<-0.0045 0.2866 -0.2766 -0.9173>}
 dyn_legStrap7_l {parent=dyn_legStrap6_l, loc=<4.967 0 0>, rot=<0 0 0 -1>}
 thigh_r {parent=pelvis, loc=<-2.5691 -0.0109 7.7501>, rot=<0.0572 -0.0012 0.9982 0.0155>}
 calf_r {parent=thigh_r, loc=<39.0487 0 0>, rot=<0.0006 -0.0161 -0.0391 -0.9991>}
 calf_twist_01_r {parent=calf_r, loc=<35.8132 0.0967 -0.0397>, rot=<0.0237 0.0095 -0.0035 -0.9997>}
 calf_twist_02_r {parent=calf_r, loc=<18.8064 0.0967 -0.0397>, rot=<0.0237 0.0095 -0.0035 -0.9997>}
 foot_r {parent=calf_r, loc=<42.5611 -0.1077 -0.4292>, rot=<0.0418 -0.0455 0.0343 -0.9975>}
 ball_r {parent=foot_r, loc=<6.1598 -13.8264 -0.6806>, rot=<0.0001 -0.0001 -0.7063 -0.7079>}
 thigh_twist_01_r {parent=thigh_r, loc=<21.6303 0 0>, rot=<0.0013 0 0 -1>}
 dyn_legStrap1_mstr_r {parent=thigh_twist_01_r, loc=<0.0058 0.0183 8.9641>, rot=<0.9983 -0.0016 0.0581 -0.0011>}
 dyn_legStrap1_r {parent=dyn_legStrap1_mstr_r, loc=<0 0 0>, rot=<0 0 0 1>}
 dyn_legStrap2_r {parent=dyn_legStrap1_r, loc=<5.4043 0 0>, rot=<-0.0004 -0.0519 -0.0102 -0.9986>}
 dyn_legStrap3_r {parent=dyn_legStrap2_r, loc=<3.2184 0 0>, rot=<0.0004 0.0208 0.0009 -0.9998>}
 dyn_legStrap4_r {parent=dyn_legStrap3_r, loc=<4.5799 0 0>, rot=<-0.0348 0.1839 0.1991 -0.9619>}
 dyn_legStrap5_r {parent=dyn_legStrap4_r, loc=<5.0391 0 0>, rot=<0.0229 -0.2033 -0.2696 -0.941>}
 dyn_legStrap6_r {parent=dyn_legStrap5_r, loc=<3.054 0 0>, rot=<0 0 0 1>}
 jacket1_mstr {parent=pelvis, loc=<-1.0765 8.8177 12.8467>, rot=<0.173 0.2835 0.9424 -0.04>}
 jacket1 {parent=jacket1_mstr, loc=<0 0 0>, rot=<0 0 0 1>}
 jacket2_mstr {parent=pelvis, loc=<-0.4907 4.9186 15.5173>, rot=<0.1832 0.1848 0.9645 -0.0459>}
 jacket2 {parent=jacket2_mstr, loc=<0 0 0>, rot=<0 0 0 1>}
 jacket3_mstr {parent=pelvis, loc=<-0.5427 -0.8905 16.3674>, rot=<0.1685 0.0038 0.9857 0.0063>}
 jacket3 {parent=jacket3_mstr, loc=<0 0 0>, rot=<0 0 0 -1>}
 jacket4_mstr {parent=pelvis, loc=<-1.1467 -6.8127 14.3384>, rot=<0.1222 -0.2695 0.9549 -0.0234>}
 jacket4 {parent=jacket4_mstr, loc=<0 0 0>, rot=<0 0 0 -1>}
 jacket5_mstr {parent=pelvis, loc=<-1.8439 -11.099 9.6865>, rot=<0.0524 -0.495 0.8672 -0.0123>}
 jacket5 {parent=jacket5_mstr, loc=<0 0 0>, rot=<0 0 0 -1>}
 Base {type=Curve, loc=<0 0 0>, rot=<0 0 0 1>}
 Pickaxe {type=Curve, loc=<0 0 0>, rot=<0 0 0 1>}
 Gun {type=Curve, loc=<0 0 0>, rot=<0 0 0 1>}
 Pistol {type=Curve, loc=<0 0 0>, rot=<0 0 0 1>}
]
