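# Physics asset for the hip pouch mesh of the Masked Dancer medium male player cosmetic, defining simulation bodies and joints for the pouch and its two zippers.
# A capsule pouch body joins a tiny master capsule through tight swing and twist limits, while front and back zipper boxes attach to the pouch with mostly looser limits.
# An unconstrained left thigh capsule is also included, apparently as a collider for the pouch.
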PhysicsAsset /BRCosmetics/Characters/Player/Male/Medium/Bodies/M_MED_Masked_Dancer/Meshes/M_MED_Masked_Dancer_HipPouch_Physics
bodies {
 dyn_backZipper [{shapeType=Box, extentX=2, extentY=1, extentZ=0.5, center=<1 0 0>}]
 dyn_frontZipper [{shapeType=Box, extentX=2, extentY=1, extentZ=0.5, center=<1 0 0>}]
 dyn_sidePouch [{shapeType=Capsule, length=13.3477, radius=11.5186, center=<7.4537 -0.1564 -2.0304>, rotation=<rot 0 0 -90>}]
 dyn_sidePouch_mstr [{shapeType=Capsule, length=0.505, radius=0.505}]
 thigh_l [{shapeType=Capsule, length=9.0775, radius=8.0826, center=<-21.6973 0.4365 -0.0306>, rotation=<rot -90 -0.0395 0.0394>}]
}
constraints [
 {bone1Name=dyn_sidePouch, bone2Name=dyn_sidePouch_mstr, swing1LimitDegrees=2, swing2LimitDegrees=10, twistLimitDegrees=1}
 {bone1Name=dyn_frontZipper, bone2Name=dyn_sidePouch, swing1LimitDegrees=15, swing2LimitDegrees=15, twistLimitDegrees=2}
 {bone1Name=dyn_backZipper, bone2Name=dyn_sidePouch, swing1LimitDegrees=15, swing2LimitDegrees=15, twistLimitDegrees=2}
]
