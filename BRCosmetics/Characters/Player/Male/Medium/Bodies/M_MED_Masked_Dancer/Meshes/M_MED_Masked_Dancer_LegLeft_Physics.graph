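# Physics asset for the left leg of the Masked Dancer male body cosmetic, giving a dangling thigh strap its physics bodies and joints.
# A master box attaches to a thigh twist capsule with free swing and twist, and seven smaller boxes chain from it.
# The first link locks twist and one swing axis, while later links share tight angular limits so the strap bends only slightly at each joint.
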PhysicsAsset /BRCosmetics/Characters/Player/Male/Medium/Bodies/M_MED_Masked_Dancer/Meshes/M_MED_Masked_Dancer_LegLeft_Physics
bodies {
 dyn_legStrap1_l [{shapeType=Box, extentX=5.5192, extentY=4.4731, extentZ=1.771, center=<2.6498 -0.1279 -0.0657>}]
 dyn_legStrap1_mstr_l [{shapeType=Box, extentX=10, extentY=10, extentZ=1, center=<5 0 -1.2>}]
 dyn_legStrap2_l [{shapeType=Box, extentX=4, extentY=3, extentZ=1, center=<2 0 0>}]
 dyn_legStrap3_l [{shapeType=Box, extentX=3, extentY=3, extentZ=1, center=<1.5 0 0>}]
 dyn_legStrap4_l [{shapeType=Box, extentX=3, extentY=3, extentZ=1, center=<1.5 0 0>}]
 dyn_legStrap5_l [{shapeType=Box, extentX=3, extentY=3, extentZ=1, center=<1.5 0 0>}]
 dyn_legStrap6_l [{shapeType=Box, extentX=5, extentY=3, extentZ=1, center=<2.5 0 0>}]
 dyn_legStrap7_l [{shapeType=Box, extentX=3, extentY=3, extentZ=1, center=<1.7341 0.0094 0.0706>}]
 thigh_twist_01_l [{shapeType=Capsule, length=19.75, radius=7.75, center=<-14.1987 0.6365 -0.249>, rotation=<rot 90 0 0>}]
}
constraints [
 {
  bone1Name=dyn_legStrap1_l
  bone2Name=dyn_legStrap1_mstr_l
  swing1LimitDegrees=5
  swing1Motion=Locked
  swing2LimitDegrees=5
  twistLimitDegrees=2
  twistMotion=Locked
 }
 {bone1Name=dyn_legStrap2_l, bone2Name=dyn_legStrap1_l, swing1LimitDegrees=8, swing2LimitDegrees=8, twistLimitDegrees=4}
 {bone1Name=dyn_legStrap3_l, bone2Name=dyn_legStrap2_l, swing1LimitDegrees=8, swing2LimitDegrees=8, twistLimitDegrees=4}
 {bone1Name=dyn_legStrap4_l, bone2Name=dyn_legStrap3_l, swing1LimitDegrees=8, swing2LimitDegrees=8, twistLimitDegrees=4}
 {bone1Name=dyn_legStrap5_l, bone2Name=dyn_legStrap4_l, swing1LimitDegrees=8, swing2LimitDegrees=8, twistLimitDegrees=4}
 {bone1Name=dyn_legStrap6_l, bone2Name=dyn_legStrap5_l, swing1LimitDegrees=8, swing2LimitDegrees=8, twistLimitDegrees=4}
 {bone1Name=dyn_legStrap7_l, bone2Name=dyn_legStrap6_l, swing1LimitDegrees=8, swing2LimitDegrees=8, twistLimitDegrees=4}
 {
  bone1Name=dyn_legStrap1_mstr_l
  bone2Name=thigh_twist_01_l
  swing1LimitDegrees=45
  swing1Motion=Free
  swing2LimitDegrees=45
  swing2Motion=Free
  twistLimitDegrees=45
  twistMotion=Free
 }
]
