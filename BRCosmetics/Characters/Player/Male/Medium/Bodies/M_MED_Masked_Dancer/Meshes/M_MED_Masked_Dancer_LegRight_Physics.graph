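# Physics asset for the right leg of the Masked Dancer cosmetic body, setting up physics for its dangling leg strap bones.
# A thigh twist capsule anchors a master strap box through an angularly free joint, leading into a chain of six strap boxes.
# The first strap locks twist and one swing axis to the master, while later links use small swing and twist limits to restrict bending.
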PhysicsAsset /BRCosmetics/Characters/Player/Male/Medium/Bodies/M_MED_Masked_Dancer/Meshes/M_MED_Masked_Dancer_LegRight_Physics
bodies {
 dyn_legStrap1_mstr_r [{shapeType=Box, extentX=10, extentY=10, extentZ=1, center=<5 0 1.2>}]
 dyn_legStrap1_r [{shapeType=Box, extentX=5.413, extentY=4.4927, extentZ=1.4097, center=<2.5843 -0.0785 0.0121>}]
 dyn_legStrap2_r [{shapeType=Box, extentX=3, extentY=4, extentZ=1, center=<1.5 0 0>}]
 dyn_legStrap3_r [{shapeType=Box, extentX=4.5, extentY=4, extentZ=1, center=<2.25 0 0>}]
 dyn_legStrap4_r [{shapeType=Box, extentX=4, extentY=4, extentZ=1, center=<2 0 0>}]
 dyn_legStrap5_r [{shapeType=Box, extentX=3, extentY=4, extentZ=1, center=<1.5 0 0>}]
 dyn_legStrap6_r [{shapeType=Box, extentX=5, extentY=4, extentZ=1, center=<2.5 0 0>}]
 thigh_twist_01_r [{shapeType=Capsule, length=17.5, radius=8, center=<12.2312 1.4311 0.1996>, rotation=<rot 90 0 0>}]
}
constraints [
 {
  bone1Name=dyn_legStrap1_r
  bone2Name=dyn_legStrap1_mstr_r
  swing1LimitDegrees=5
  swing1Motion=Locked
  swing2LimitDegrees=5
  twistLimitDegrees=2
  twistMotion=Locked
 }
 {bone1Name=dyn_legStrap2_r, bone2Name=dyn_legStrap1_r, swing1LimitDegrees=9, swing2LimitDegrees=9, twistLimitDegrees=4}
 {bone1Name=dyn_legStrap3_r, bone2Name=dyn_legStrap2_r, swing1LimitDegrees=9, swing2LimitDegrees=9, twistLimitDegrees=4}
 {bone1Name=dyn_legStrap4_r, bone2Name=dyn_legStrap3_r, swing1LimitDegrees=9, swing2LimitDegrees=9, twistLimitDegrees=4}
 {bone1Name=dyn_legStrap5_r, bone2Name=dyn_legStrap4_r, swing1LimitDegrees=9, swing2LimitDegrees=9, twistLimitDegrees=4}
 {bone1Name=dyn_legStrap6_r, bone2Name=dyn_legStrap5_r, swing1LimitDegrees=9, swing2LimitDegrees=9, twistLimitDegrees=4}
 {
  bone1Name=dyn_legStrap1_mstr_r
  bone2Name=thigh_twist_01_r
  swing1LimitDegrees=45
  swing1Motion=Free
  swing2LimitDegrees=45
  swing2Motion=Free
  twistLimitDegrees=45
  twistMotion=Free
 }
]
