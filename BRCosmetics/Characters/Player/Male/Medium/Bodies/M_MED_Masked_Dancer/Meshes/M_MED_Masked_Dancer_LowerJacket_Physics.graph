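# Physics asset for the lower jacket of the male medium Masked Dancer cosmetic body, providing capsule bodies for jacket physics.
# Five jacket bones each join a tiny master-bone capsule through tight swing and twist limits.
# The master bones join the pelvis with free angular motion, as does the right upper thigh twist bone.
# Larger pelvis and right thigh capsules give the body volume; there is no left thigh body.
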PhysicsAsset /BRCosmetics/Characters/Player/Male/Medium/Bodies/M_MED_Masked_Dancer/Meshes/M_MED_Masked_Dancer_LowerJacket_Physics
bodies {
 jacket1 [{shapeType=Capsule, length=9.875, radius=4.625, shapeName=jacket1_capsule, center=<8.9271 0 3.5>, rotation=<rot 82.3554 -90.0008 -0.0006>}]
 jacket1_mstr [{shapeType=Capsule, length=0.505, radius=0.505, rotation=<rot 0 -90 0>}]
 jacket2 [{shapeType=Capsule, length=9.5, radius=4.75, shapeName=jacket2_capsule, center=<8.1397 0 3.5>, rotation=<rot 88.8451 90.0041 -179.9957>}]
 jacket2_mstr [{shapeType=Capsule, length=0.505, radius=0.505, rotation=<rot 0 -90 0>}]
 jacket3 [{shapeType=Capsule, length=9.75, radius=4.5, shapeName=jacket3_capsule, center=<7 0 3>, rotation=<rot 86.7227 -90.0003 -0.0001>}]
 jacket3_mstr [{shapeType=Capsule, length=0.505, radius=0.505, rotation=<rot 0 -90 0>}]
 jacket4 [{shapeType=Capsule, length=9.75, radius=4.5, shapeName=jacket4_capsule, center=<8.1397 0 3.5>, rotation=<rot 86.9855 89.9946 179.995>}]
 jacket4_mstr [{shapeType=Capsule, length=0.505, radius=0.505, rotation=<rot 0 -90 0>}]
 jacket5 [{shapeType=Capsule, length=9, radius=5.25, shapeName=jacket5_capsule, center=<7 0 4>, rotation=<rot 87.9539 80.2633 179.9985>}]
 jacket5_mstr [{shapeType=Capsule, length=0.505, radius=0.505, rotation=<rot 0 -90 0>}]
 pelvis [{shapeType=Capsule, length=7.8972, radius=8.874, shapeName=pelvis_capsule, center=<-3.5511 -2.1928 0>}]
 thigh_twist_01_r [{shapeType=Capsule, length=20.9546, radius=8.742, center=<-7.5494 -0.6285 -0.2044>, rotation=<rot 90 0 0>}]
}
constraints [
 {bone1Name=jacket1, bone2Name=jacket1_mstr, swing1LimitDegrees=15, swing2LimitDegrees=15, twistLimitDegrees=1}
 {bone1Name=jacket2, bone2Name=jacket2_mstr, swing1LimitDegrees=15, swing2LimitDegrees=15, twistLimitDegrees=1}
 {bone1Name=jacket3, bone2Name=jacket3_mstr, swing1LimitDegrees=10, swing2LimitDegrees=10, twistLimitDegrees=1}
 {bone1Name=jacket4, bone2Name=jacket4_mstr, swing1LimitDegrees=10, swing2LimitDegrees=10, twistLimitDegrees=1}
 {bone1Name=jacket5, bone2Name=jacket5_mstr, swing1LimitDegrees=10, swing2LimitDegrees=10, twistLimitDegrees=1}
 {
  bone1Name=thigh_twist_01_r
  bone2Name=pelvis
  swing1LimitDegrees=45
  swing1Motion=Free
  swing2LimitDegrees=45
  swing2Motion=Free
  twistLimitDegrees=45
  twistMotion=Free
 }
 {
  bone1Name=jacket1_mstr
  bone2Name=pelvis
  swing1LimitDegrees=45
  swing1Motion=Free
  swing2LimitDegrees=45
  swing2Motion=Free
  twistLimitDegrees=45
  twistMotion=Free
 }
 {
  bone1Name=jacket2_mstr
  bone2Name=pelvis
  swing1LimitDegrees=45
  swing1Motion=Free
  swing2LimitDegrees=45
  swing2Motion=Free
  twistLimitDegrees=45
  twistMotion=Free
 }
 {
  bone1Name=jacket3_mstr
  bone2Name=pelvis
  swing1LimitDegrees=45
  swing1Motion=Free
  swing2LimitDegrees=45
  swing2Motion=Free
  twistLimitDegrees=45
  twistMotion=Free
 }
 {
  bone1Name=jacket4_mstr
  bone2Name=pelvis
  swing1LimitDegrees=45
  swing1Motion=Free
  swing2LimitDegrees=45
  swing2Motion=Free
  twistLimitDegrees=45
  twistMotion=Free
 }
 {
  bone1Name=jacket5_mstr
  bone2Name=pelvis
  swing1LimitDegrees=45
  swing1Motion=Free
  swing2LimitDegrees=45
  swing2Motion=Free
  twistLimitDegrees=45
  twistMotion=Free
 }
]
