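# Physics asset for the medium male Masked Dancer body cosmetic, covering a dangling chest strap and the upper spine.
# Large capsules on spine_04 and spine_05 form the torso volume, with small capsules along the strap bones.
# A three-link strap chain hangs from a master bone constrained to spine_05.
# The master and spine joints rotate freely, while strap links have a narrow swing range on one axis only.
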PhysicsAsset /BRCosmetics/Characters/Player/Male/Medium/Bodies/M_MED_Masked_Dancer/Meshes/M_MED_Masked_Dancer_Physics
bodies {
 dyn_chestStrap1 [{shapeType=Capsule, length=4.596, radius=0.5, shapeName=dyn_chestStrap1_capsule, center=<3.6623 0 0>, rotation=<rot 90 0 0>}]
 dyn_chestStrap2 [{shapeType=Capsule, length=4.8566, radius=0.6413, shapeName=dyn_chestStrap2_capsule, center=<1 0 0>, rotation=<rot 90 0 0>}]
 dyn_chestStrap3 [{shapeType=Capsule, length=7.2722, radius=0.9921, shapeName=dyn_chestStrap3_capsule, center=<4.122 0 0>, rotation=<rot 90 0 0>}]
 dyn_chestStrap_mstr [{shapeType=Capsule, length=0.505, radius=0.505, shapeName=dyn_chestStrap_mstr_capsule, rotation=<rot 0 -90 0>}]
 spine_04 [{shapeType=Capsule, length=17.1405, radius=12.0586, shapeName=spine_04_capsule, center=<1.1034 1.2086 0.4897>}]
 spine_05 [{shapeType=Capsule, length=15.8743, radius=10.7675, shapeName=spine_05_capsule, center=<-2.1363 0.0929 -1.3612>}]
}
constraints [
 {
  bone1Name=dyn_chestStrap_mstr
  bone2Name=spine_05
  swing1LimitDegrees=45
  swing1Motion=Free
  swing2LimitDegrees=45
  swing2Motion=Free
  twistLimitDegrees=45
  twistMotion=Free
 }
 {bone1Name=dyn_chestStrap1, bone2Name=dyn_chestStrap_mstr, swing1LimitDegrees=5, swing2LimitDegrees=0, twistLimitDegrees=0}
 {bone1Name=dyn_chestStrap2, bone2Name=dyn_chestStrap1, swing1LimitDegrees=5, swing2LimitDegrees=0, twistLimitDegrees=0}
 {bone1Name=dyn_chestStrap3, bone2Name=dyn_chestStrap2, swing1LimitDegrees=5, swing2LimitDegrees=0, twistLimitDegrees=0}
 {
  bone1Name=spine_05
  bone2Name=spine_04
  swing1LimitDegrees=45
  swing1Motion=Free
  swing2LimitDegrees=45
  swing2Motion=Free
  twistLimitDegrees=45
  twistMotion=Free
 }
]
